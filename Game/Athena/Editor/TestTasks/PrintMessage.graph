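# PrintMessage is an Editor Utility Blueprint task, stored among editor test tasks, that prints a message when it starts and when it finishes.
# BeginExecution prints "Start " plus the Message string, waits two seconds through a Delay, prints "Finish " plus the same string, then calls FinishExecutingTask.
# The Delay keeps the task running over time, so completion is signaled only after the wait.
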EditorUtilityBlueprint /Game/Athena/Editor/TestTasks/PrintMessage
graphs [EventGraph {nodes [
   Event BeginExecution()
   > PrintString(InString=Append(A="Start ", B=Get Message() -> (Message:String)))
   > Delay!(execute, WorldContextObject:Object Reference, Duration=2) -> (then)
   > PrintString(InString=Append(A="Finish ", B=Get Message()))
   > FinishExecutingTask()
  ]}]
variables [Message {type=string, default=""}]
