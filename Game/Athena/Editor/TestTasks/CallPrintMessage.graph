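# CallPrintMessage is an editor utility test task that runs three PrintMessage tasks as children of itself.
# On BeginExecution it constructs and registers each child with messages 1 to 3, chaining on the immediate exec output without waiting for completion.
# Only the third child's OnFinished leads on to a short delay, an "All Done" print and FinishExecutingTask.
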EditorUtilityBlueprint /Game/Athena/Editor/TestTasks/CallPrintMessage
graphs [EventGraph {nodes [
   Event BeginExecution()
   n1 = Self-Reference()
   > n2 = Construct Object from Class!(execute, Class=&/PrintMessage.PrintMessage_C, self=@n1, Message:String=1) -> (then, ReturnValue:Print Message Object Reference)
   > RegisterAndExecuteTask!(execute:Exec, Task:Editor Utility Task Object Reference=@n2, OptionalParentTask:Editor Utility Task Object Reference=@n1) -> (then:Exec, OnFinished:Exec, Task:Editor Utility Task Object Reference) {is=title K2Node_AsyncAction}
   n3 = Self-Reference()
   > n4 = Construct Object from Class(Class=&PrintMessage_C, self=@n3, Message=2)
   > RegisterAndExecuteTask(Task=@n4, OptionalParentTask=@n3)
   n5 = Self-Reference()
   > n6 = Construct Object from Class(Class=&PrintMessage_C, self=@n5, Message=3)
   > n7 = RegisterAndExecuteTask(Task=@n6, OptionalParentTask=@n5)
   Delay!(execute=@n7.OnFinished, WorldContextObject:Object Reference, Duration=2) -> (then)
   > PrintString(InString=All Done)
   > FinishExecutingTask()
  ]}]
variables []
